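# Niagara system for SquadLock's exit-volume effect, named for playing on exit: a brief red flash with scanning planes.
# BlockyStuff_onExit bursts a short-lived GPU sprite whose size, brightness and alpha follow curves.
# Scan_Planes_onExit spawns upward-facing sprites, gated by loop age, that oscillate within User.BboxSize with staggered timing, fading from red toward near-black.
# A third halo emitter, single_halo_onExit, is disabled.
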NiagaraSystem /SquadLockGameplay/items/SquadLock/Assets/FX/NS_SquadLock_ExitVolume
dependencies {
 usedDataInterfaces [&NiagaraDataInterfaceCurve, &NiagaraDataInterfaceColorCurve]
 usedDynamicInputs []
 usedModules [&SystemState]
}
emitters {
 BlockyStuff_onExit {
  simTarget=GPUComputeSim
  emitterUpdateScript [
   &EmitterState(Inactive Response=::NewEnumerator1, Life Cycle Mode=::NewEnumerator1, Loop Behavior=::NewEnumerator1, Loop Duration=1.5, UseLoopDelay=false)
   &SpawnBurst_Instantaneous(Use Spawn Probability=false, UseLoopCountLimit=false)
  ]
  particleSpawnScript [&InitializeParticle(
    Color=<rgba 0.6 0 0 1>
    Lifetime=0.2
    Sprite Size Mode=::NewEnumerator3
    Uniform Sprite Size=111
    UsePositionOffset=false
    Write Material Random=true
   )]
  particleUpdateScript [
   &ParticleState()
   /SetVariables(Particles.Position=&Add_Vector(A=<0 0 0>, B=<0 0 25>)) {isSetParameters=true, module=SetVariables_#1}
   /SetVariables(Particles.SpriteSize=&Multiply_Vector2D_ByFloat(
     Float=&FloatFromCurve(
      FloatCurve=DataInterface() {propertyValues {Curve {keys [
          {arriveTangent=-0.0094, interpMode=RCIM_Cubic, leaveTangent=-0.0094, tangentMode=RCTM_User, time=0, value=2}
          {arriveTangent=3.9425, interpMode=RCIM_Cubic, leaveTangent=3.9425, tangentMode=RCTM_User, time=0.3, value=3.25}
          {arriveTangent=-0.0405, interpMode=RCIM_Cubic, leaveTangent=-0.0405, tangentMode=RCTM_User, time=1, value=4.375}
         ]}}}
      Scale Curve=15
     )
     Vector2D=<1 1>
    )) {isSetParameters=true, module=SetVariables_#2}
   &ScaleColor(
    Scale Alpha=&FloatFromCurve(
     FloatCurve=DataInterface() {propertyValues {Curve {keys [
         {arriveTangent=-0.2933, interpMode=RCIM_Cubic, leaveTangent=-0.2933, tangentMode=RCTM_User, time=0, value=0}
         {arriveTangent=-2.8313, interpMode=RCIM_Cubic, leaveTangent=-2.8313, tangentMode=RCTM_User, time=0.1, value=0.95}
         {arriveTangent=-0.9489, interpMode=RCIM_Cubic, leaveTangent=-0.9489, tangentMode=RCTM_User, time=0.2, value=0.25}
         {arriveTangent=-0.0383, interpMode=RCIM_Cubic, leaveTangent=-0.0383, tangentMode=RCTM_User, time=1, value=0}
        ]}}}
     Scale Curve=2
    )
    Scale RGB=&Multiply_VectorByFloat(
     Vector=<1 1 1>
     Float=&FloatFromCurve(FloatCurve=DataInterface() {propertyValues {Curve {keys [
          {arriveTangent=-0.2933, interpMode=RCIM_Cubic, leaveTangent=-0.2933, leaveTangentWeight=0.0206, tangentMode=RCTM_User, time=0, value=0}
          {arriveTangent=49.2109, interpMode=RCIM_Cubic, leaveTangent=49.2109, tangentMode=RCTM_User, time=0.0869, value=4.3724}
          {
           arriveTangent=-25.8313
           arriveTangentWeight=0.0198
           interpMode=RCIM_Cubic
           leaveTangent=-25.8313
           leaveTangentWeight=0.0079
           tangentMode=RCTM_User
           time=0.1
           value=4.75
          }
          {
           arriveTangent=-0.6853
           arriveTangentWeight=0.0012
           interpMode=RCIM_Cubic
           leaveTangent=-0.6853
           leaveTangentWeight=0.1655
           tangentMode=RCTM_User
           time=0.5327
           value=0.4397
          }
          {arriveTangent=-0.8945, arriveTangentWeight=0.1656, interpMode=RCIM_Cubic, leaveTangent=-0.8945, tangentMode=RCTM_User, time=1, value=0}
         ]}}})
    )
    ScaleA=true
    ScaleRGB=true
   ) {module=ScaleColor001}
   &CameraOffset(Camera Offset Amount=25) {module=CameraOffset001}
   /SetVariables(
    Particles.Color_B:LinearColor=&MultiplyLinearColor(Linear Color=$Particles.Color, Scale Factor=&MakeVector4_FromFloat(Float=0.3))
    Particles.SpriteSize_B:Vector2f=&Multiply_Vector2D_ByFloat(Float=1, Vector2D=$Particles.SpriteSize)
   ) {enabled=false, isSetParameters=true}
  ]
  rendererClasses [&NiagaraSpriteRendererProperties]
  renderers [{rendererClass=&NiagaraSpriteRendererProperties}]
 }
 Scan_Planes_onExit {
  emitterUpdateScript [
   &EmitterState(Life Cycle Mode=::NewEnumerator1, Loop Behavior=::NewEnumerator1, Loop Duration=1, UseLoopDelay=false)
   &SpawnRate(
    Spawn Probability=&MakeCustomFloatFromBool(
     Bool=&SetBoolByFloatComparison(A=$Emitter.NormalizedLoopAge, B=0.3, Comparison Type=::NewEnumerator6)
     True Float=1
     False Float=0
    )
    SpawnRate=55
    Use Spawn Probability=true
   )
  ]
  particleSpawnScript [
   &InitializeParticle(
    Color Mode=::NewEnumerator0
    Lifetime=0.3
    Mass Mode=::NewEnumerator1
    Position Mode=::NewEnumerator1
    Sprite Size=<75 75>
    Sprite Size Mode=::NewEnumerator4
    UsePositionOffset=true
    Write Material Random=true
   )
   /SetVariables(Particles.Position=&Add_Vector(A=<0 0 0>, B=<0 0 55>)) {isSetParameters=true}
  ]
  particleUpdateScript [
   &ParticleState()
   &ScaleColor(
    Scale Alpha=&FloatFromCurve(FloatCurve=DataInterface() {propertyValues {Curve {keys [
         {arriveTangent=8.6044, interpMode=RCIM_Cubic, leaveTangent=8.6044, leaveTangentWeight=0.0349, tangentMode=RCTM_User, time=-0.0046, value=0}
         <at 0.1 0.9>
         {
          arriveTangent=-1.3396
          arriveTangentWeight=0.2
          interpMode=RCIM_Cubic
          leaveTangent=-1.3396
          leaveTangentWeight=0.0666
          tangentMode=RCTM_User
          time=0.3
          value=0.35
         }
         {arriveTangent=0.0723, interpMode=RCIM_Cubic, leaveTangent=0.0723, tangentMode=RCTM_User, time=1.0011, value=0.002}
        ]}}})
    ScaleA=true
    ScaleRGB=true
   )
   &SpriteFacingAndAlignment(Facing Coordinate Space=::World, Sprite Facing=<0 0 1>, WriteAlignment=false, WriteFacing=true)
   &self:OscillateInBounds_01(
    BboxSize:Vector3f=$User.BboxSize
    InitialPosition:Vector3f=<0 0 35>
    InputTime:NiagaraFloat=$Emitter.Age
    Magnitude=1
    Speed:NiagaraFloat=&RandomRangeFloat(Minimum=0.1, Maximum=0.4)
    Stagger:NiagaraFloat=&ReturnNormalizedExecIndex(Normalized Index Scale=0.55)
   ) {module=OscillateInBounds_01}
   &CameraOffset(Camera Offset Amount=15)
   &Color(Color=&LerpLinearColors(
     StartColor=<rgba 1 0 0 1>
     EndColor=<rgba 0 0.01 0 1>
     LerpFactor=&FloatFromCurve(FloatCurve=DataInterface() {propertyValues {Curve {keys [
          {arriveTangent=0.1837, interpMode=RCIM_Cubic, leaveTangent=0.1837, tangentMode=RCTM_User, time=0, value=0}
          {arriveTangent=-0.2496, interpMode=RCIM_Cubic, leaveTangent=-0.2496, tangentMode=RCTM_User, time=1, value=1}
         ]}}})
    ))
   /SetVariables(Particles.Color=&ScaleLinearColorByCurve(
     InitialValue=$Particles.Color
     ScaleCurve=DataInterface() {propertyValues {
       AlphaCurve {keys []}
       BlueCurve {keys [<at 0 1>,<at 0.0441 1>,<at 0.5584 0.1042>,<at 1 0>]}
       GreenCurve {keys [<at 0 1>,<at 0.0441 1>,<at 0.5584 0.1042>,<at 1 0>]}
       RedCurve {keys [<at 0 1>,<at 0.0441 1>,<at 0.5584 0.1042>,<at 1 0>]}
      }}
     CurveIndex=$Particles.Oscillate
    )) {isSetParameters=true}
  ]
  rendererClasses [&NiagaraSpriteRendererProperties]
  renderers [{rendererClass=&NiagaraSpriteRendererProperties}]
 }
 single_halo_onExit {
  bEnabled=false
  simTarget=GPUComputeSim
  emitterUpdateScript [
   &EmitterState(Inactive Response=::NewEnumerator1, Life Cycle Mode=::NewEnumerator1, Loop Behavior=::NewEnumerator1, Loop Duration=1.5, UseLoopDelay=false)
   &SpawnBurst_Instantaneous(Use Spawn Probability=false, UseLoopCountLimit=false)
  ]
  particleSpawnScript [&InitializeParticle(
    Color=<rgba 25 0 0 1>
    Lifetime=0.75
    Position Mode=::NewEnumerator1
    Sprite Size Mode=::NewEnumerator3
    Uniform Sprite Size=111
    UsePositionOffset=false
    Write Material Random=true
   )]
  particleUpdateScript [
   &ParticleState()
   /SetVariables(Particles.Position=&Add_Vector(A=<0 0 0>, B=<0 0 25>)) {enabled=false, isSetParameters=true, module=SetVariables_#1}
   /SetVariables(Particles.SpriteSize=&Multiply_Vector2D_ByFloat(
     Float=&FloatFromCurve(
      FloatCurve=DataInterface() {propertyValues {Curve {keys [
          {arriveTangent=-0.0094, interpMode=RCIM_Cubic, leaveTangent=-0.0094, tangentMode=RCTM_User, time=0, value=0}
          {arriveTangent=7.6747, interpMode=RCIM_Cubic, leaveTangent=7.6747, tangentMode=RCTM_User, time=0.3, value=2}
          {arriveTangent=-0.0405, interpMode=RCIM_Cubic, leaveTangent=-0.0405, tangentMode=RCTM_User, time=1, value=4.375}
         ]}}}
      Scale Curve=35
     )
     Vector2D=<1 1>
    )) {isSetParameters=true, module=SetVariables_#2}
   &FN_Radial_Material(Sprite Extrusion Speed=0.1)
   &ScaleColor(
    Scale Alpha=&FloatFromCurve(
     FloatCurve=DataInterface() {propertyValues {Curve {keys [
         {arriveTangent=-0.2933, interpMode=RCIM_Cubic, leaveTangent=-0.2933, tangentMode=RCTM_User, time=0, value=0}
         {arriveTangent=-2.8313, interpMode=RCIM_Cubic, leaveTangent=-2.8313, tangentMode=RCTM_User, time=0.1, value=0.95}
         {arriveTangent=-1.6622, interpMode=RCIM_Cubic, leaveTangent=-1.6622, tangentMode=RCTM_User, time=0.2, value=0.45}
         {arriveTangent=-0.0383, interpMode=RCIM_Cubic, leaveTangent=-0.0383, tangentMode=RCTM_User, time=1, value=0}
        ]}}}
     Scale Curve=2
    )
    ScaleA=true
    ScaleRGB=false
   ) {module=ScaleColor001}
   &CameraOffset(Camera Offset Amount=25) {module=CameraOffset001}
  ]
  rendererClasses [&NiagaraSpriteRendererProperties]
  renderers [
   {rendererClass=&NiagaraSpriteRendererProperties}
   {rendererClass=&NiagaraSpriteRendererProperties, rendererIndex=1}
   {rendererClass=&NiagaraSpriteRendererProperties, rendererIndex=2}
  ]
 }
}
summary {
 systemName=NS_SquadLock_ExitVolume
 emitters [
  {bEnabled=false, emitterName=single_halo_onExit, simTarget=GPUComputeSim, rendererClasses [&NiagaraSpriteRendererProperties]}
  {emitterName=Scan_Planes_onExit, rendererClasses [&NiagaraSpriteRendererProperties]}
  {emitterName=BlockyStuff_onExit, simTarget=GPUComputeSim, rendererClasses [&NiagaraSpriteRendererProperties]}
 ]
 userVariables [
  User.ActorCenter {type=Vector3f, description="", defaultValue {struct=Vector3f, value=<0 0 0>}}
  User.AgeInput {type=NiagaraFloat, defaultValue=0, description=""}
  User.BboxSize {type=Vector3f, description="", defaultValue {struct=Vector3f, value=<0 0 55>}}
  User.ScannerCenter {type=Vector3f, description="", defaultValue {struct=Vector3f, value=<0 0 0>}}
 ]
}
