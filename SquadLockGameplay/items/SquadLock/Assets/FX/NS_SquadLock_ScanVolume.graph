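# Niagara effect for the SquadLock item that draws two scan beams fanning out from User.ScannerCenter.
# Each ribbon emitter bursts two particles, lerped by exec index between that center and an oscillating fixed offset, widening from zero by center-to-owner distance.
# User.AgeInput scales particle age, letting external input influence the curve-driven blue-to-red color shift.
# The LockLight mesh emitter and BboxSize-based oscillation are disabled, and the ribbons' sine pulse remaps to constant brightness.
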NiagaraSystem /SquadLockGameplay/items/SquadLock/Assets/FX/NS_SquadLock_ScanVolume
dependencies {usedDataInterfaces [&NiagaraDataInterfaceCurve], usedDynamicInputs [], usedModules [&SystemState]}
emitters {
 LockLight {
  bEnabled=false
  emitterUpdateScript [
   &EmitterState(Life Cycle Mode=::NewEnumerator1, Loop Behavior=::NewEnumerator1, Loop Duration=1, UseLoopDelay=false)
   &SpawnBurst_Instantaneous(Use Spawn Probability=false, UseLoopCountLimit=false)
  ]
  particleSpawnScript [
   &InitializeParticle(
    Color Mode=::NewEnumerator0
    Lifetime=0.1
    Mesh Scale Mode=::NewEnumerator3
    Mesh Uniform Scale=0.2
    UsePositionOffset=false
    Write Material Random=true
   )
   /SetVariables() {isSetParameters=true}
  ]
  particleUpdateScript [
   &ParticleState()
   /SetVariables(Particles.Age:NiagaraFloat=&Multiply_Float(A=$Particles.Age, B=&Multiply_Float(A=$User.AgeInput, B=1))) {isSetParameters=true}
   /SetVariables(Particles.MeshOrientation=&MakeQuatFromAxisAngle(Axis=<0 0 1>, AngleInDegrees=-90)) {isSetParameters=true}
   &Color(Color=&LerpLinearColors(
     StartColor=<rgba 0 0.2 1.3 1>
     EndColor=<rgba 0.9583 0.0176 0 1>
     LerpFactor=&FloatFromCurve(FloatCurve=DataInterface() {propertyValues {Curve {keys [<at 0.3 0>,<at 0.35 1>]}}})
    )) {module=Color001}
   /SetVariables(Particles.Color=&MultiplyLinearColor(
     Linear Color=$Particles.Color
     Scale Factor=&MakeVector4_FromFloat(Float=&RemapRange(
       Input Value To Remap=&Multiply_Float(A=&Sine(Normalized Angle=$Emitter.Age, Period=2, Scale=1, Bias=5), B=1)
       Input Min=-1
       Input Max=1
       Output Min=0.3
       Output Max=1
      ))
    )) {isSetParameters=true}
  ]
  rendererClasses [&NiagaraMeshRendererProperties]
  renderers [{rendererClass=&NiagaraMeshRendererProperties, rendererIndex=0}]
 }
 Scan_Ribbon002 {
  emitterUpdateScript [
   &EmitterState(Life Cycle Mode=::NewEnumerator1, Loop Behavior=::NewEnumerator1, Loop Duration=1, UseLoopDelay=false)
   &SpawnBurst_Instantaneous(Spawn Count=2, Spawn Group=1, Use Spawn Probability=false, UseLoopCountLimit=false) {module=SpawnBurst_Instantaneous001}
  ]
  particleSpawnScript [
   &InitializeRibbon(
    Color=<rgba 0 1 0 1>
    Lifetime=0.5
    Position Mode:ENiagara_PositionInitializationMode=::NewEnumerator0
    Ribbon Facing Mode:ENiagara_UnsetDirectSet=::NewEnumerator1
    Ribbon Width=55
    Write Material Random=false
   )
   /SetVariables(Particles.ScannerCenter:Vector3f=&Add_Vector(A=$User.ScannerCenter, B=<3 -15 2>)) {enabled=false, isSetParameters=true, module=SetVariables_#1}
   /SetVariables(Particles.InitialPosition:Vector3f=<0 -125 0>) {enabled=false, isSetParameters=true, module=SetVariables_#2}
   /SetVariables(Particles.ScannerCenter:Vector3f=&Add_Vector(A=$User.ScannerCenter, B=<0 0 0>)) {isSetParameters=true}
   /SetVariables(Particles.InitialPosition:Vector3f=<0 -125 -55>) {isSetParameters=true}
  ]
  particleUpdateScript [
   &ParticleState()
   /SetVariables(Particles.Age:NiagaraFloat=&Multiply_Float(A=$Particles.Age, B=&Multiply_Float(A=$User.AgeInput, B=1))) {isSetParameters=true, module=SetVariables_#3}
   &self:OscillatePosition_01(
    BboxSize:Vector3f=$User.BboxSize
    InputTime:NiagaraFloat=$Emitter.Age
    Magnitude=0.2
    Speed=0.3
    Stagger:NiagaraFloat=&ReturnNormalizedExecIndex(Normalized Index Scale=0.55)
   ) {enabled=false, module=OscillatePosition_01}
   &self:Oscillate_simple(Magnitude=12, Speed=0.2) {module=Oscillate_simple}
   /SetVariables(Particles.Position=&Lerp_Vector(
     A=$Particles.ScannerCenter
     B=&Add_Vector(A=$Particles.Oscillate, B=$Particles.InitialPosition)
     Alpha=&ReturnNormalizedExecIndex()
    )) {isSetParameters=true, module=SetVariables_#4}
   /SetVariables(Particles.RibbonWidth=&Lerp_Float(
     A=0
     B=&Multiply_Float(
      A=&VectorLength(Vector=&Subtract_Vector(A=$Particles.ScannerCenter, B=Linked() {linkedVariable {name=Engine.Owner.Position, type=NiagaraPosition}}))
      B=0.005
     )
     Alpha=&ReturnNormalizedExecIndex()
    )) {isSetParameters=true, module=SetVariables_#5}
   &Color(Color=&LerpLinearColors(
     StartColor=<rgba 0 0.2 1.3 1>
     EndColor=<rgba 1 0 0 1>
     LerpFactor=&FloatFromCurve(FloatCurve=DataInterface() {propertyValues {Curve {keys [<at 0.05 0>,<at 0.2 1>]}}})
    )) {module=Color001}
   &Color(Color=&LerpLinearColors(
     StartColor=<rgba 0 0.3 0 0.3>
     EndColor=<rgba 0.3 0 0 1>
     LerpFactor=&FloatFromCurve(FloatCurve=DataInterface() {propertyValues {Curve {keys [<at 0.1 0>,<at 0.3 1>]}}})
    )) {enabled=false}
   /SetVariables(Particles.Color=&MultiplyLinearColor(
     Linear Color=$Particles.Color
     Scale Factor=&MakeVector4_FromFloat(Float=&RemapRange(
       Input Value To Remap=&Multiply_Float(A=&Sine(Normalized Angle=$Emitter.Age, Period=1, Scale=1, Bias=0), B=1)
       Input Min=-1
       Input Max=1
       Output Min=1
       Output Max=1
      ))
    )) {isSetParameters=true, module=SetVariables_#6}
  ]
  rendererClasses [&NiagaraRibbonRendererProperties]
  renderers [
   {rendererClass=&NiagaraRibbonRendererProperties}
   {rendererClass=&NiagaraRibbonRendererProperties, rendererIndex=1}
  ]
 }
 Scan_Ribbon003 {
  emitterUpdateScript [
   &EmitterState(Life Cycle Mode=::NewEnumerator1, Loop Behavior=::NewEnumerator1, Loop Duration=1, UseLoopDelay=false)
   &SpawnBurst_Instantaneous(Spawn Count=2, Spawn Group=1, Use Spawn Probability=false, UseLoopCountLimit=false) {module=SpawnBurst_Instantaneous001}
  ]
  particleSpawnScript [
   &InitializeRibbon(
    Color=<rgba 0 1 0 1>
    Lifetime=0.5
    Position Mode:ENiagara_PositionInitializationMode=::NewEnumerator0
    Ribbon Facing Mode:ENiagara_UnsetDirectSet=::NewEnumerator1
    Ribbon Width=55
    Write Material Random=false
   )
   /SetVariables(Particles.ScannerCenter:Vector3f=&Add_Vector(A=$User.ScannerCenter, B=<3 -15 2>)) {enabled=false, isSetParameters=true, module=SetVariables_#1}
   /SetVariables(Particles.InitialPosition:Vector3f=<0 -125 0>) {enabled=false, isSetParameters=true, module=SetVariables_#2}
   /SetVariables(Particles.ScannerCenter:Vector3f=&Add_Vector(A=$User.ScannerCenter, B=<0 0 0>)) {isSetParameters=true}
   /SetVariables(Particles.InitialPosition:Vector3f=<0 -125 -155>) {isSetParameters=true}
  ]
  particleUpdateScript [
   &ParticleState()
   /SetVariables(Particles.Age:NiagaraFloat=&Multiply_Float(A=$Particles.Age, B=&Multiply_Float(A=$User.AgeInput, B=1))) {isSetParameters=true, module=SetVariables_#3}
   &self:OscillatePosition_01(
    BboxSize:Vector3f=$User.BboxSize
    InputTime:NiagaraFloat=$Emitter.Age
    Magnitude=0.2
    Speed=0.3
    Stagger:NiagaraFloat=&ReturnNormalizedExecIndex(Normalized Index Scale=0.55)
   ) {enabled=false, module=OscillatePosition_01}
   &self:Oscillate_simple(Magnitude=12, Speed=0.1) {module=Oscillate_simple}
   /SetVariables(Particles.Position=&Lerp_Vector(
     A=$Particles.ScannerCenter
     B=&Add_Vector(A=$Particles.Oscillate, B=$Particles.InitialPosition)
     Alpha=&ReturnNormalizedExecIndex()
    )) {isSetParameters=true, module=SetVariables_#4}
   /SetVariables(Particles.RibbonWidth=&Lerp_Float(
     A=0
     B=&Multiply_Float(
      A=&VectorLength(Vector=&Subtract_Vector(A=$Particles.ScannerCenter, B=Linked() {linkedVariable {name=Engine.Owner.Position, type=NiagaraPosition}}))
      B=0.005
     )
     Alpha=&ReturnNormalizedExecIndex()
    )) {isSetParameters=true, module=SetVariables_#5}
   &Color(Color=&LerpLinearColors(
     StartColor=<rgba 0 0.2 1.3 1>
     EndColor=<rgba 0.3 0 0 1>
     LerpFactor=&FloatFromCurve(FloatCurve=DataInterface() {propertyValues {Curve {keys [<at 0.05 0>,<at 0.2 1>]}}})
    )) {module=Color001}
   &Color(Color=&LerpLinearColors(
     StartColor=<rgba 0 0.3 0 0.3>
     EndColor=<rgba 0.3 0 0 1>
     LerpFactor=&FloatFromCurve(FloatCurve=DataInterface() {propertyValues {Curve {keys [<at 0.1 0>,<at 0.3 1>]}}})
    )) {enabled=false}
   /SetVariables(Particles.Color=&MultiplyLinearColor(
     Linear Color=$Particles.Color
     Scale Factor=&MakeVector4_FromFloat(Float=&RemapRange(
       Input Value To Remap=&Multiply_Float(A=&Sine(Normalized Angle=$Emitter.Age, Period=2, Scale=1, Bias=5), B=1)
       Input Min=-1
       Input Max=1
       Output Min=1
       Output Max=1
      ))
    )) {isSetParameters=true, module=SetVariables_#6}
  ]
  rendererClasses [&NiagaraRibbonRendererProperties]
  renderers [
   {rendererClass=&NiagaraRibbonRendererProperties}
   {rendererClass=&NiagaraRibbonRendererProperties, rendererIndex=1}
  ]
 }
}
summary {
 systemName=NS_SquadLock_ScanVolume
 emitters [
  {emitterName=Scan_Ribbon002, rendererClasses [&NiagaraRibbonRendererProperties]}
  {emitterName=Scan_Ribbon003, rendererClasses [&NiagaraRibbonRendererProperties]}
  {bEnabled=false, emitterName=LockLight, rendererClasses [&NiagaraMeshRendererProperties]}
 ]
 userVariables [
  User.AgeInput {type=NiagaraFloat, defaultValue=0, description=""}
  User.BboxSize {type=Vector3f, description="", defaultValue {struct=Vector3f, value=<0 0 0>}}
  User.ScannerCenter {type=Vector3f, description="", defaultValue {struct=Vector3f, value=<0 0 0>}}
 ]
}
