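# Niagara burst effect for a SquadLock vault opening, mostly green, built from short-lived emitters.
# Mesh distortion, shockwave flare and pulsing lock light combine with sprite glows, a tall beam, torus-placed radial flares, warm sparkles and a light renderer.
# Curves drive scale, color and material parameters; halo and sparkles dim by time of day, and the halo also follows music amplitude.
# The GPU OmnidirectionalBurst emitter is disabled.
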NiagaraSystem /SquadLockGameplay/items/SquadLock/Assets/FX/NS_SquadLock_VaultOpen
dependencies {
 usedDataInterfaces [
  &NiagaraDataInterfaceCurve
  &NiagaraDataInterfaceColorCurve
  &NiagaraDataInterfaceVectorCurve
  &NiagaraDataInterfaceVector2DCurve
 ]
 usedDynamicInputs []
 usedModules [&SystemState]
}
emitters {
 Distortion {
  emitterUpdateScript [
   &EmitterState(Life Cycle Mode=::NewEnumerator1, Loop Behavior=::NewEnumerator1, Loop Delay=0.025, Loop Duration=1, UseLoopDelay=true)
   &SpawnBurst_Instantaneous(Use Spawn Probability=false, UseLoopCountLimit=false)
  ]
  particleSpawnScript [
   /Niagara/Modules/Spawn/Initialization/InitializeParticle(
    Lifetime=0.35
    Mesh Scale=<5 5 5>
    Module.Write Color=true
    Module.Write Lifetime=true
    Module.Write Mass=true
    Module.Write Material Random=true
    Module.Write Position=false
    Module.Write Scale=true
    Module.Write SpriteRotation=false
    Module.Write SpriteSize=false
   )
   &Color(Color=&MakeLinearColorFromVectorAndFloat("Vector (RGB)"=<0.2491 0.7031 1>, "Float (Alpha)"=&ClampFloat(Float=0)))
  ]
  particleUpdateScript [
   &ParticleState()
   /Niagara/Modules/Update/Orientation/V2/OrientMeshToVector(Look At Direction:Vector3f=<0 1 0>, Override Quat=false, UseOptionalSideAxis=false, UseOptionalUpAxis=false)
   &ScaleMeshSize(Scale Factor=&VectorFromFloat(Value=&FloatFromCurve(
      FloatCurve=DataInterface() {propertyValues {Curve {keys [
          {arriveTangent=6.638, interpMode=RCIM_Cubic, time=0, value=0}
          {arriveTangent=1, interpMode=RCIM_Cubic, leaveTangent=1, time=0.2381, value=0.8455}
          <at 1 1>
         ]}}}
      Scale Curve=0.3
     )))
   &DynamicMaterialParameters(
    Index 0 Param 1=&FloatFromCurve(FloatCurve=DataInterface() {propertyValues {Curve {keys [{arriveTangent=0.9145, interpMode=RCIM_Cubic, time=0.2, value=0.88}, <at 1 1>]}}})
    Index 0 Param 2=&FloatFromCurve(FloatCurve=DataInterface() {propertyValues {Curve {keys [<at 0 1.25 RCIM_Cubic>, {arriveTangent=-3.093, leaveTangent=-3.093, time=1, value=0}]}}})
    Index 0 Param 3=&FloatFromCurve(FloatCurve=DataInterface() {propertyValues {Curve {keys [{arriveTangent=0.6796, interpMode=RCIM_Cubic, time=0, value=-0.5}, <at 1 0>]}}})
    Index 0 Param 4=&FloatFromCurve(FloatCurve=DataInterface() {propertyValues {Curve {keys [{arriveTangent=0.2861, interpMode=RCIM_Cubic, time=0, value=0}, <at 1 0.12>]}}})
    Module.Param0WriteEnabled=true
    Module.Param1WriteEnabled=true
    Module.Param2WriteEnabled=true
    Module.Param3WriteEnabled=true
   )
   &ScaleColor(
    Scale Alpha=&FloatFromCurve(FloatCurve=DataInterface() {propertyValues {Curve {keys [<at 0 1 RCIM_Cubic>, {arriveTangent=-2.3476, leaveTangent=-2.3476, time=1, value=0}]}}})
    Scale RGB=&VectorFromFloat(Value=&FloatFromCurve(FloatCurve=DataInterface() {propertyValues {Curve {keys [{arriveTangent=-23.8521, interpMode=RCIM_Cubic, time=0, value=10}, <at 1 1>]}}}))
    ScaleA=true
    ScaleRGB=true
   )
  ]
  rendererClasses [&NiagaraMeshRendererProperties]
  renderers [{rendererClass=&NiagaraMeshRendererProperties, rendererIndex=0}]
 }
 Halo_and_Glow {
  emitterUpdateScript [
   &EmitterState()
   &SpawnRate(Use Spawn Probability=false) {enabled=false}
   &SpawnBurst_Instantaneous(Spawn Count=1, Use Spawn Probability=false)
  ]
  particleSpawnScript [/Niagara/Modules/Spawn/Initialization/V2/InitializeParticle(Color Mode=::NewEnumerator0, Lifetime=0.5, UsePositionOffset=true, Write Material Random=true)]
  particleUpdateScript [
   &ParticleState()
   &Color(
    Color=&ColorFromCurve(ColorCurve=DataInterface() {propertyValues {
       AlphaCurve {keys [<at 0.1791 2 RCIM_Cubic>,<at 0.9745 0 RCIM_Cubic>]}
       BlueCurve {keys [<at 0.2314 0.3642>, {arriveTangent=-0.9419, interpMode=RCIM_Cubic, leaveTangent=-0.9419, time=0.6814, value=0}]}
       GreenCurve {keys [<at 0.2314 3>, {arriveTangent=-5.563, interpMode=RCIM_Cubic, leaveTangent=-5.563, time=0.6814, value=0.849}]}
       RedCurve {keys [<at 0.2314 0>, {arriveTangent=1.3256, interpMode=RCIM_Cubic, leaveTangent=1.3256, time=0.6814, value=0.5126}]}
      }})
    Scale Color=<0.05 0.05 0.05>
   )
   /SetVariables(Particles.SpriteSize=&Vector2DFromFloat(Value=&FloatFromCurve(
      FloatCurve=DataInterface() {propertyValues {Curve {keys [
          {arriveTangent=0.3463, interpMode=RCIM_Cubic, leaveTangent=0.3463, tangentMode=RCTM_User, time=0, value=0.2}
          {arriveTangent=1.8357, interpMode=RCIM_Cubic, leaveTangent=1.8357, tangentMode=RCTM_User, time=0.3, value=0.95}
          {arriveTangent=-0.0405, interpMode=RCIM_Cubic, leaveTangent=-0.0405, tangentMode=RCTM_User, time=1, value=1.5}
         ]}}}
      Scale Curve=155
     ))) {isSetParameters=true}
   &CameraOffset(Camera Offset Amount=55)
   &FN_Radial_Material(
    Radial Mask Power=1
    Radial Mask Tile U=0.2
    Radial Mask Tile V=0.5
    Sprite Extrusion Speed:NiagaraFloat=&FloatFromCurve(
     FloatCurve=DataInterface() {propertyValues {Curve {keys [
         {arriveTangent=-0.2256, interpMode=RCIM_Cubic, time=0, value=1.75}
         {arriveTangent=2.75, interpMode=RCIM_Cubic, leaveTangent=2.75, time=0.3, value=2.375}
         {arriveTangent=2.5, interpMode=RCIM_Cubic, leaveTangent=2.5, time=0.5, value=3.125}
         {interpMode=RCIM_Cubic, leaveTangent=-0.5666, time=0.7, value=3.375}
        ]}}}
     Scale Curve=-0.3
    )
    Sprite Spin Speed:NiagaraFloat=&FloatFromCurve(FloatCurve=DataInterface() {propertyValues {Curve {keys [<at 0 0.2>,<at 0.4 1.125>,<at 1 2.25>]}}}, Scale Curve=0.1)
   ) {enabled=false}
  ]
  rendererClasses [&NiagaraSpriteRendererProperties]
  renderers [
   {rendererClass=&NiagaraSpriteRendererProperties}
   {rendererClass=&NiagaraSpriteRendererProperties, rendererIndex=1}
  ]
 }
 LightRenderer {
  emitterUpdateScript [
   &EmitterState(Life Cycle Mode=::NewEnumerator1, Loop Behavior=::NewEnumerator1, Loop Duration=2, UseLoopDelay=false)
   &SpawnBurst_Instantaneous(Use Spawn Probability=false, UseLoopCountLimit=false)
  ]
  particleSpawnScript [
   &InitializeParticle(
    Color=<rgba 0.2923 1 0 15>
    Lifetime=0.5
    Sprite Size Mode=::NewEnumerator1
    Uniform Sprite Size Max=100
    Uniform Sprite Size Min=55
    UsePositionOffset=false
    Write Material Random=true
   )
   &SphereLocation(
    MaskSpawnGroup=false
    Override Seed=false
    Sphere Radius=2
    Surface Only=false
    UseNonUniformScale=false
    UseOffset=true
    UseSphereOrientation=false
   )
   /SetVariables(Particles.LightRadius=100, Particles.LightScale:Vector3f=<1 1 1>) {isSetParameters=true}
  ]
  particleUpdateScript [
   &ParticleState()
   &ScaleColor(
    Scale Alpha=&FloatFromCurve(FloatCurve=DataInterface() {propertyValues {Curve {keys [<at 0.2 2>,<at 0.7 0.5>,<at 1 0>]}}})
    ScaleA=true
    ScaleRGB=true
   )
   &CameraOffset(Camera Offset Amount=55)
  ]
  rendererClasses [&NiagaraLightRendererProperties]
  renderers [{rendererClass=&NiagaraLightRendererProperties, rendererIndex=0}]
 }
 LockLight {
  emitterUpdateScript [
   &EmitterState(Life Cycle Mode=::NewEnumerator1, Loop Behavior=::NewEnumerator1, Loop Duration=1, UseLoopDelay=false)
   &SpawnBurst_Instantaneous(Use Spawn Probability=false, UseLoopCountLimit=false)
  ]
  particleSpawnScript [
   &InitializeParticle(
    Color Mode=::NewEnumerator0
    Lifetime=0.5
    Mesh Scale Mode=::NewEnumerator3
    Mesh Uniform Scale=0.2
    UsePositionOffset=false
    Write Material Random=true
   )
   /SetVariables() {isSetParameters=true}
  ]
  particleUpdateScript [
   &ParticleState()
   /SetVariables(Particles.Age:NiagaraFloat=&Multiply_Float(A=$Particles.Age, B=&Multiply_Float(A=$User.AgeInput, B=1))) {enabled=false, isSetParameters=true}
   /SetVariables(Particles.MeshOrientation=&MakeQuatFromAxisAngle(Axis=<0 0 1>, AngleInDegrees=-90)) {isSetParameters=true}
   &Color(Color=&LerpLinearColors(
     StartColor=<rgba 0 1.3 0.0692 1>
     EndColor=<rgba 0 0 0 1>
     LerpFactor=&FloatFromCurve(FloatCurve=DataInterface() {propertyValues {Curve {keys [<at 0.8 0>,<at 1 1>]}}})
    )) {module=Color001}
   /SetVariables(Particles.Color=&MultiplyLinearColor(
     Linear Color=$Particles.Color
     Scale Factor=&MakeVector4_FromFloat(Float=&RemapRange(
       Input Value To Remap=&Multiply_Float(A=&Sine(Normalized Angle=$Emitter.Age, Period=2, Scale=1, Bias=5), B=1)
       Input Min=-1
       Input Max=1
       Output Min=0.3
       Output Max=1
      ))
    )) {isSetParameters=true}
  ]
  rendererClasses [&NiagaraMeshRendererProperties]
  renderers [{rendererClass=&NiagaraMeshRendererProperties, rendererIndex=0}]
 }
 OmnidirectionalBurst {
  bEnabled=false
  simTarget=GPUComputeSim
  emitterUpdateScript [
   &EmitterState(Life Cycle Mode=::NewEnumerator1, Loop Behavior=::NewEnumerator1, Loop Duration=1, UseLoopDelay=false)
   &SpawnBurst_Instantaneous(Spawn Count=35, Use Spawn Probability=false, UseLoopCountLimit=false)
  ]
  particleSpawnScript [
   &InitializeParticle(
    Color Maximum=<rgba 1 1 0 1>
    Color Minimum=<rgba 0 1 0.3 1>
    Color Mode=::NewEnumerator2
    Lifetime Min=0.3
    Lifetime Mode=::NewEnumerator1
    Mass Max=5
    Mass Min=0.4
    Mass Mode=::NewEnumerator3
    Sprite Rotation Mode=::NewEnumerator2
    Sprite Size Mode=::NewEnumerator1
    Uniform Sprite Size Max=12
    Uniform Sprite Size Min=8
    UsePositionOffset=false
    Write Material Random=true
   )
   &SphereLocation(
    MaskSpawnGroup=false
    Offset=<0 0 15>
    Offset Coordinate Space:ENiagaraCoordinateSpace=::World
    Sphere Distribution:ENiagaraSphereDistributionMode=::NewEnumerator3
    Sphere Origin:NiagaraPosition=$Engine.Owner.Position
    Sphere Radius=125
    UseNonUniformScale=false
    UseOffset=true
    UseSphereOrientation=false
   )
   &AddVelocityFromPoint(
    Module.Use Falloff Distance=false
    Origin Offset=<0 0 -55>
    Origin Offset Coordinate Space=::World
    Velocity Origin:Vector3f=/Niagara/DynamicInputs/Helpers/SimulationPosition()
    Velocity Strength=&RandomRangeFloat(Minimum=444, Maximum=222)
   )
  ]
  particleUpdateScript [
   &ParticleState()
   &GravityForce(Gravity=<0 0 -255>)
   &Drag(Drag=&RandomRangeFloat(Minimum=5, Maximum=2), Use Linear Drag=true, Use Rotational Drag=false)
   &ScaleSpriteSize(Scale Factor=&Vector2DFromFloat(Value=&FloatFromCurve(
      FloatCurve=DataInterface() {propertyValues {Curve {keys [
          <at -0.0046 1.0113 RCIM_Cubic>
          {arriveTangent=-1.0036, interpMode=RCIM_Cubic, leaveTangent=-1.0036, time=0.7001, value=0.872}
          <at 1.0011 0.002>
         ]}}}
      Scale Curve=&RandomRangeFloat(Minimum=0.3, Maximum=0.4)
     )))
   &ScaleColor(
    Scale Alpha=&FloatFromCurve(FloatCurve=DataInterface() {propertyValues {Curve {keys [<at -0.0046 1.0113 RCIM_Cubic>,<at 1.0011 0.002 RCIM_Cubic>]}}})
    Scale RGB=&VectorFromFloat(Value=&FloatFromCurve(FloatCurve=DataInterface() {propertyValues {Curve {keys [<at 0 1>,<at 1 0>]}}}, Scale Curve=24))
    ScaleA=true
    ScaleRGB=true
   )
   &CurlNoiseForce(Noise Frequency=12555, Noise Strength=444, Pan Noise=true, Pan Noise Field=<0 125 155>)
   &SolveForcesAndVelocity(Clamp Velocity=false, Limit Acceleration=false)
   &CameraOffset(Camera Offset Amount=55)
  ]
  rendererClasses [&NiagaraSpriteRendererProperties]
  renderers [{rendererClass=&NiagaraSpriteRendererProperties}]
 }
 Shockwave_Flare {
  emitterUpdateScript [
   &EmitterState(Life Cycle Mode=::NewEnumerator1, Loop Behavior=::NewEnumerator1, Loop Duration=1, UseLoopDelay=true)
   &SpawnBurst_Instantaneous(Use Spawn Probability=false, UseLoopCountLimit=false)
  ]
  particleSpawnScript [
   /Niagara/Modules/Spawn/Initialization/InitializeParticle(
    Lifetime=0.5
    Mesh Scale=<5 5 5>
    Module.Write Color=true
    Module.Write Lifetime=true
    Module.Write Mass=true
    Module.Write Material Random=true
    Module.Write Position=true
    Module.Write Scale=true
    Module.Write SpriteRotation=false
    Module.Write SpriteSize=false
   )
   &self:Position() {module=Position}
   &Color(Color=<rgba 0 1 0.1 1>) {module=Color001}
  ]
  particleUpdateScript [
   &ParticleState()
   &OrientMeshToVector(Look At Direction:Vector3f=<1 0 0>, Override Quat=false, UseOptionalSideAxis=false, UseOptionalUpAxis=false)
   &DynamicMaterialParameters(
    Index 0 Param 1=&FloatFromCurve(FloatCurve=DataInterface() {propertyValues {Curve {keys [
         {arriveTangent=4.7478, interpMode=RCIM_Cubic, time=-0.0042, value=0.0106}
         {arriveTangent=-0.0105, interpMode=RCIM_Cubic, leaveTangent=-0.0105, time=0.2286, value=0.5}
         <at 1 0>
        ]}}})
    Index 0 Param 2=&FloatFromCurve(FloatCurve=DataInterface() {propertyValues {Curve {keys [<at 0 15 RCIM_Cubic>,<at 1 0>]}}})
    Index 0 Param 3=&FloatFromCurve(FloatCurve=DataInterface() {propertyValues {Curve {keys [<at 0 -0.8>,<at 1 -0.8>]}}})
    Index 0 Param 4=&FloatFromCurve(FloatCurve=DataInterface() {propertyValues {Curve {keys [<at 0 0>,<at 1 -1>]}}})
    Module.Param0WriteEnabled=true
    Module.Param1WriteEnabled=true
    Module.Param2WriteEnabled=true
    Module.Param3WriteEnabled=true
   )
   &ScaleMeshSize(Scale Factor=&VectorFromFloat(Value=&FloatFromCurve(
      FloatCurve=DataInterface() {propertyValues {Curve {keys [
          {arriveTangent=8.8797, interpMode=RCIM_Cubic, time=-0.0014, value=0}
          {arriveTangent=2.0056, interpMode=RCIM_Cubic, leaveTangent=2.0056, time=0.2495, value=1.3042}
          {interpMode=RCIM_Cubic, leaveTangent=0.4549, time=1, value=2.0084}
         ]}}}
      Scale Curve=0.0851
     )))
   &ScaleColor(
    Scale Alpha=&FloatFromCurve(
     FloatCurve=DataInterface() {propertyValues {Curve {keys [
         <at -0.0043 2.2383 RCIM_Cubic>
         {arriveTangent=-4.9832, interpMode=RCIM_Cubic, leaveTangent=-4.9832, time=0.0972, value=1}
         {arriveTangent=-1.1077, interpMode=RCIM_Cubic, leaveTangent=-1.1077, time=0.3875, value=0.2861}
         <at 1 0>
        ]}}}
     Scale Curve=0.2
    )
    Scale RGB=&VectorFromCurve(VectorCurve=DataInterface() {propertyValues {
       XCurve {keys [
         <at 0 7.5967 RCIM_Cubic>
         {arriveTangent=-5.9799, interpMode=RCIM_Cubic, leaveTangent=-5.9799, time=0.1, value=2.3911}
         <at 1 1.6167 RCIM_Cubic>
        ]}
       YCurve {keys [
         <at 0 9.342 RCIM_Cubic>
         {arriveTangent=-8.8732, interpMode=RCIM_Cubic, leaveTangent=-8.8732, time=0.1, value=4.8507}
         <at 1 0.4688 RCIM_Cubic>
        ]}
       ZCurve {keys [<at 0 15 RCIM_Cubic>, {arriveTangent=-13, interpMode=RCIM_Cubic, leaveTangent=-13, time=0.1, value=5}, <at 1 2 RCIM_Cubic>]}
      }})
    ScaleA=true
    ScaleRGB=true
   )
   &CameraOffset(Camera Offset Amount=125)
  ]
  rendererClasses [&NiagaraMeshRendererProperties]
  renderers [{rendererClass=&NiagaraMeshRendererProperties, rendererIndex=0}]
 }
 TallBeam_sprite {
  emitterUpdateScript [
   &EmitterState(Life Cycle Mode=::NewEnumerator1, Loop Behavior=::NewEnumerator1, Loop Duration=1, UseLoopDelay=true)
   &SpawnBurst_Instantaneous(Use Spawn Probability=false, UseLoopCountLimit=false)
  ]
  particleSpawnScript [
   &InitializeParticle(
    Color=<rgba 0 1 0 1>
    Module.Write Color=false
    Module.Write Lifetime=true
    Module.Write Mass=true
    Module.Write Material Random=true
    Module.Write Position=true
    Module.Write Scale=true
    Module.Write SpriteRotation=true
    Module.Write SpriteSize=true
    Position=<0 0 0>
    Sprite Size=<100 175>
   )
   &InitialMeshOrientation(Mesh Orientation Mode=::NewEnumerator2, Orientation Vector=<0 0 1>, Use Rotation Vector=false)
  ]
  particleUpdateScript [
   &ParticleState()
   &ScaleColor(
    Scale Alpha=&FloatFromCurve(FloatCurve=DataInterface() {propertyValues {Curve {keys [<at 0 0 RCIM_Cubic>,<at 0.0468 1 RCIM_Cubic>,<at 1 0>]}}})
    Scale RGB=&Multiply_VectorByFloat(
     Vector=<1 1 1>
     Float=&FloatFromCurve(
      FloatCurve=DataInterface() {propertyValues {Curve {keys [
          {arriveTangent=9.0725, interpMode=RCIM_Cubic, leaveTangent=9.0725, tangentMode=RCTM_User, time=0, value=0}
          {arriveTangent=-2.5249, interpMode=RCIM_Cubic, leaveTangent=-2.5249, tangentMode=RCTM_User, time=0.1408, value=0.9847}
          {arriveTangent=-0.0405, interpMode=RCIM_Cubic, leaveTangent=-0.0405, tangentMode=RCTM_User, time=1, value=0}
         ]}}}
      Scale Curve=1.3
     )
    )
    ScaleA=true
    ScaleRGB=true
   )
   &DynamicMaterialParameters(
    Index 0 Param 1=&FloatFromCurve(
     FloatCurve=DataInterface() {propertyValues {Curve {keys [{arriveTangent=2294.5679, interpMode=RCIM_Cubic, time=0, value=0}, <at 0.4 280>]}}}
     Scale Curve=0.3
    )
    Index 0 Param 2=1
    Index 0 Param 3=1
    Index 0 Param 4=1
    Module.Param0WriteEnabled=true
    Module.Param1WriteEnabled=true
    Module.Param2WriteEnabled=true
    Module.Param3WriteEnabled=true
   )
   &ScaleSpriteSize(
    Non-Uniform Curve Scale=<10 10>
    Non-Uniform Curve Sprite Scale=DataInterface() {propertyValues {
      XCurve {keys [{interpMode=RCIM_Cubic, leaveTangentWeight=0.3333, tangentMode=RCTM_User, time=0, value=0.3}, <at 0.4 0.3>]}
      YCurve {keys [{interpMode=RCIM_Cubic, tangentMode=RCTM_User, time=0, value=0}, <at 0.4 1>]}
     }}
    Scale Sprite Size Mode=::NewEnumerator3
   )
   &CameraOffset(Camera Offset Amount=55)
  ]
  rendererClasses [&NiagaraSpriteRendererProperties]
  renderers [{rendererClass=&NiagaraSpriteRendererProperties}]
 }
 radial_flares {
  emitterUpdateScript [
   &EmitterState(Inactive Response=::NewEnumerator1, Life Cycle Mode=::NewEnumerator1, Loop Behavior=::NewEnumerator1, Loop Duration=1, UseLoopDelay=false)
   &SpawnBurst_Instantaneous(Spawn Count=12, Use Spawn Probability=false, UseLoopCountLimit=false)
  ]
  particleSpawnScript [/Niagara/Modules/Spawn/Initialization/V2/InitializeParticle(Color=<rgba 0.3 1 0 0.15>, Sprite Size=<2 35>, Sprite Size Mode=::NewEnumerator4, UsePositionOffset=false, Write Material Random=true)]
  particleUpdateScript [
   &ParticleState()
   &TorusLocation(
    Align Torus With Axis=true
    Handle Radius=5
    Large Radius:NiagaraFloat=&FloatFromCurve(
     FloatCurve=DataInterface() {propertyValues {Curve {keys [
         {arriveTangent=9.0725, interpMode=RCIM_Cubic, leaveTangent=9.0725, leaveTangentWeight=0.3879, tangentMode=RCTM_User, time=0, value=0}
         {
          arriveTangent=7.6747
          arriveTangentWeight=0.1
          interpMode=RCIM_Cubic
          leaveTangent=7.6747
          leaveTangentWeight=0.2333
          tangentMode=RCTM_User
          time=0.2
          value=2.125
         }
         {
          arriveTangent=7.6747
          arriveTangentWeight=0.0333
          interpMode=RCIM_Cubic
          leaveTangent=7.6747
          leaveTangentWeight=0.2333
          tangentMode=RCTM_User
          time=0.3
          value=2.75
         }
         {arriveTangent=-0.0405, arriveTangentWeight=0.2335, interpMode=RCIM_Cubic, leaveTangent=-0.0405, tangentMode=RCTM_User, time=1, value=4.375}
        ]}}}
     Scale Curve=&RandomRangeFloat(Minimum=12, Maximum=16)
    )
    MaskSpawnGroup=false
    Normalized Angle Around Torus Axis=1
    Torus Axis=<0 1 0>
    Torus Distribution Mode:ENiagaraTorusDistributionMode=::NewEnumerator1
    Torus Origin:NiagaraPosition=&VectorFromFloat(Value=0)
    U Position:NiagaraFloat=&Add_Float(
     A=&ReturnNormalizedExecIndex()
     B=&ModuloFloat(A=&Multiply_Float(A=$Particles.Age, B=0.001), B=1)
    )
   )
   &self:ScratchModule_04() {module=ScratchModule_04}
   /SetVariables(Particles.Scale=&VectorFromFloat(Value=1)) {isSetParameters=true, module=SetVariables_#1}
   &ScaleColor(
    Scale Alpha=&FloatFromCurve(FloatCurve=DataInterface() {propertyValues {Curve {keys [
         {arriveTangent=-0.2933, interpMode=RCIM_Cubic, leaveTangent=-0.2933, tangentMode=RCTM_User, time=0, value=0}
         {arriveTangent=-2.2474, interpMode=RCIM_Cubic, leaveTangent=-2.2474, tangentMode=RCTM_User, time=0.2, value=0.75}
         {arriveTangent=0.0301, interpMode=RCIM_Cubic, leaveTangent=0.0301, tangentMode=RCTM_User, time=1, value=0}
        ]}}})
    Scale RGB=&Multiply_VectorByFloat(
     Vector=&MakeVectorFromLinearColorRGB(Color=$Particles.Color)
     Float=&FloatFromCurve(FloatCurve=DataInterface() {propertyValues {Curve {keys [
          {arriveTangent=-0.2933, interpMode=RCIM_Cubic, leaveTangent=-0.2933, leaveTangentWeight=0.3333, tangentMode=RCTM_User, time=0, value=0}
          {
           arriveTangent=-4.8422
           arriveTangentWeight=0.0667
           interpMode=RCIM_Cubic
           leaveTangent=-4.8422
           leaveTangentWeight=0.2667
           tangentMode=RCTM_User
           time=0.1
           value=0.8
          }
          {arriveTangent=-0.8259, interpMode=RCIM_Cubic, leaveTangent=-0.8259, tangentMode=RCTM_User, time=0.3334, value=0.1639}
          {arriveTangent=0.0301, arriveTangentWeight=0.2667, interpMode=RCIM_Cubic, leaveTangent=0.0301, tangentMode=RCTM_User, time=1, value=0}
         ]}}})
    )
    ScaleA=true
    ScaleRGB=true
   ) {module=ScaleColor001}
   &ScaleSpriteSize(
    Uniform Curve Scale=10
    Uniform Curve Sprite Scale=DataInterface() {propertyValues {Curve {keys [
        {arriveTangent=-0.2933, interpMode=RCIM_Cubic, leaveTangent=-0.2933, leaveTangentWeight=0.4283, tangentMode=RCTM_User, time=0, value=0}
        {
         arriveTangent=0.0624
         arriveTangentWeight=0.1274
         interpMode=RCIM_Cubic
         leaveTangent=0.0624
         leaveTangentWeight=0.7778
         tangentMode=RCTM_User
         time=0.2
         value=0.7
        }
        {arriveTangent=0.0301, arriveTangentWeight=0.2866, interpMode=RCIM_Cubic, leaveTangent=0.0301, tangentMode=RCTM_User, time=1, value=0}
       ]}}}
   )
   /SetVariables(Particles.SpriteSize_B:Vector2f=<33 125>) {enabled=false, isSetParameters=true, module=SetVariables_#2}
  ]
  rendererClasses [&NiagaraSpriteRendererProperties]
  renderers [
   {rendererClass=&NiagaraSpriteRendererProperties}
   {rendererClass=&NiagaraSpriteRendererProperties, rendererIndex=1}
  ]
 }
 radial_flares001 {
  emitterUpdateScript [
   &EmitterState(Inactive Response=::NewEnumerator1, Life Cycle Mode=::NewEnumerator1, Loop Behavior=::NewEnumerator1, Loop Duration=0.3, UseLoopDelay=false)
   &SpawnBurst_Instantaneous(Spawn Count=35, Use Spawn Probability=false, UseLoopCountLimit=false)
  ]
  particleSpawnScript [&InitializeParticle(Color=<rgba 0.75 3 0 0.6>, Sprite Size=<1 15>, Sprite Size Mode=::NewEnumerator4, UsePositionOffset=false, Write Material Random=true)]
  particleUpdateScript [
   &ParticleState()
   &TorusLocation(
    Align Torus With Axis=true
    Handle Radius=5
    Large Radius:NiagaraFloat=&FloatFromCurve(
     FloatCurve=DataInterface() {propertyValues {Curve {keys [
         {arriveTangent=9.0725, interpMode=RCIM_Cubic, leaveTangent=9.0725, leaveTangentWeight=0.3879, tangentMode=RCTM_User, time=0, value=0}
         {
          arriveTangent=7.6747
          arriveTangentWeight=0.1
          interpMode=RCIM_Cubic
          leaveTangent=7.6747
          leaveTangentWeight=0.2333
          tangentMode=RCTM_User
          time=0.2
          value=2.125
         }
         {
          arriveTangent=7.6747
          arriveTangentWeight=0.0333
          interpMode=RCIM_Cubic
          leaveTangent=7.6747
          leaveTangentWeight=0.2333
          tangentMode=RCTM_User
          time=0.3
          value=2.75
         }
         {arriveTangent=-0.0405, arriveTangentWeight=0.2335, interpMode=RCIM_Cubic, leaveTangent=-0.0405, tangentMode=RCTM_User, time=1, value=4.375}
        ]}}}
     Scale Curve=&RandomRangeFloat(Minimum=15, Maximum=33)
    )
    MaskSpawnGroup=false
    Normalized Angle Around Torus Axis=1
    Torus Axis=<0 1 0>
    Torus Distribution Mode:ENiagaraTorusDistributionMode=::NewEnumerator1
    U Position:NiagaraFloat=&Add_Float(
     A=&ReturnNormalizedExecIndex()
     B=&ModuloFloat(A=&Multiply_Float(A=$Particles.Age, B=0.001), B=1)
    )
   )
   &self:ScratchModule_05() {module=ScratchModule_05}
   /SetVariables(Particles.Scale=&VectorFromFloat(Value=1)) {enabled=false, isSetParameters=true, module=SetVariables_#1}
   &ScaleColor(
    Scale Alpha=&FloatFromCurve(
     FloatCurve=DataInterface() {propertyValues {Curve {keys [
         {arriveTangent=-0.2933, interpMode=RCIM_Cubic, leaveTangent=-0.2933, tangentMode=RCTM_User, time=0, value=0}
         {arriveTangent=-3.5801, interpMode=RCIM_Cubic, leaveTangent=-3.5801, tangentMode=RCTM_User, time=0.1, value=1}
         {arriveTangent=0.0301, interpMode=RCIM_Cubic, leaveTangent=0.0301, tangentMode=RCTM_User, time=0.9, value=0.1}
        ]}}}
     Scale Curve=0.075
    )
    ScaleA=true
    ScaleRGB=false
   ) {module=ScaleColor001}
   &ScaleSpriteSize(
    Uniform Curve Scale=15
    Uniform Curve Sprite Scale=DataInterface() {propertyValues {Curve {keys [
        {arriveTangent=-0.2933, interpMode=RCIM_Cubic, leaveTangent=-0.2933, leaveTangentWeight=0.4283, tangentMode=RCTM_User, time=0, value=0}
        {
         arriveTangent=-3.5801
         arriveTangentWeight=0.1274
         interpMode=RCIM_Cubic
         leaveTangent=-3.5801
         leaveTangentWeight=0.7778
         tangentMode=RCTM_User
         time=0.1408
         value=0.9847
        }
        {arriveTangent=0.0301, arriveTangentWeight=0.2866, interpMode=RCIM_Cubic, leaveTangent=0.0301, tangentMode=RCTM_User, time=1, value=0}
       ]}}}
   )
   /SetVariables(Particles.SpriteSize_B:Vector2f=<33 125>) {isSetParameters=true, module=SetVariables_#2}
  ]
  rendererClasses [&NiagaraSpriteRendererProperties]
  renderers [{rendererClass=&NiagaraSpriteRendererProperties}]
 }
 single_halo001 {
  emitterUpdateScript [
   &EmitterState(Inactive Response=::NewEnumerator1, Life Cycle Mode=::NewEnumerator1, Loop Behavior=::NewEnumerator1, Loop Duration=1.5, UseLoopDelay=false)
   &SpawnBurst_Instantaneous(Use Spawn Probability=false, UseLoopCountLimit=false)
  ]
  particleSpawnScript [&InitializeParticle(
    Color=<rgba 2 5 0 1>
    Lifetime=0.3
    Sprite Size Mode=::NewEnumerator3
    Uniform Sprite Size=75
    UsePositionOffset=false
    Write Material Random=true
   )]
  particleUpdateScript [
   &ParticleState()
   &self:PositionNoise_01() {enabled=false, module=PositionNoise_01}
   &CameraOffset(Camera Offset Amount=15)
   /SetVariables(Particles.SpriteSize=&Multiply_Vector2D_ByFloat(
     Float=&FloatFromCurve(
      FloatCurve=DataInterface() {propertyValues {Curve {keys [
          {arriveTangent=9.0725, interpMode=RCIM_Cubic, leaveTangent=9.0725, tangentMode=RCTM_User, time=0, value=0}
          {arriveTangent=7.6747, interpMode=RCIM_Cubic, leaveTangent=7.6747, tangentMode=RCTM_User, time=0.3, value=2.75}
          {arriveTangent=-0.0405, interpMode=RCIM_Cubic, leaveTangent=-0.0405, tangentMode=RCTM_User, time=1, value=3.125}
         ]}}}
      Scale Curve=75
     )
     Vector2D=<1 1>
    )) {enabled=false, isSetParameters=true}
   /SetVariables(Particles.SpriteSize=&Multiply_Vector2D_ByFloat(
     Float=&FloatFromCurve(
      FloatCurve=DataInterface() {propertyValues {Curve {keys [
          {arriveTangent=-0.0094, interpMode=RCIM_Cubic, leaveTangent=-0.0094, tangentMode=RCTM_User, time=0, value=0}
          {arriveTangent=7.6747, interpMode=RCIM_Cubic, leaveTangent=7.6747, tangentMode=RCTM_User, time=0.3, value=2}
          {arriveTangent=-0.0405, interpMode=RCIM_Cubic, leaveTangent=-0.0405, tangentMode=RCTM_User, time=1, value=4.375}
         ]}}}
      Scale Curve=50
     )
     Vector2D=<1 1>
    )) {isSetParameters=true}
   &ScaleColor(
    Scale Alpha=&FloatFromCurve(FloatCurve=DataInterface() {propertyValues {Curve {keys [
         {arriveTangent=-0.2933, interpMode=RCIM_Cubic, leaveTangent=-0.2933, tangentMode=RCTM_User, time=0, value=0}
         {arriveTangent=-2.8313, interpMode=RCIM_Cubic, leaveTangent=-2.8313, tangentMode=RCTM_User, time=0.1, value=0.95}
         {arriveTangent=-1.6622, interpMode=RCIM_Cubic, leaveTangent=-1.6622, tangentMode=RCTM_User, time=0.2, value=0.45}
         {arriveTangent=-0.0383, interpMode=RCIM_Cubic, leaveTangent=-0.0383, tangentMode=RCTM_User, time=1, value=0}
        ]}}})
    ScaleA=true
    ScaleRGB=false
   ) {module=ScaleColor001}
   /SetVariables(Particles.SpriteSize_B:Vector2f=&Multiply_Vector2D_ByFloat(
     Float=&FloatFromCurve(FloatCurve=DataInterface() {propertyValues {Curve {keys [
          {arriveTangent=22.1, interpMode=RCIM_Cubic, leaveTangent=22.1, tangentMode=RCTM_User, time=0, value=0}
          {arriveTangent=-6.4477, interpMode=RCIM_Cubic, leaveTangent=-6.4477, tangentMode=RCTM_User, time=0.1, value=1}
          {arriveTangent=-0.0405, interpMode=RCIM_Cubic, leaveTangent=-0.0405, tangentMode=RCTM_User, time=0.5, value=0}
         ]}}})
     Vector2D=<0.3 5>
    )) {enabled=false, isSetParameters=true}
   /SetVariables(Particles.Scale=&VectorFromFloat(Value=25)) {enabled=false, isSetParameters=true, module=SetVariables_#1}
   &ScaleColor(
    Scale Alpha=&Multiply_Float(
     A=&FloatFromCurve(FloatCurve=DataInterface() {propertyValues {Curve {keys [<at 0.8 1>,<at 1 0>]}}})
     B=&FNTimeOfDayScalar(Night Scale=0.3, Morning Scale=0.1, Evening Scale=0.2)
    )
    Scale RGB=&VectorFromFloat(Value=&FloatFromCurve(
      FloatCurve=DataInterface() {propertyValues {Curve {keys [<at 0 1>,<at 0.1 3>,<at 0.2 1>]}}}
      Scale Curve=&Lerp_Float(A=1, B=3, Alpha=$NPC.NPC_Audio_Reactive_Parameters.Amplitude_Music)
     ))
    ScaleA=true
    ScaleRGB=true
   )
  ]
  rendererClasses [&NiagaraSpriteRendererProperties]
  renderers [
   {rendererClass=&NiagaraSpriteRendererProperties}
   {rendererClass=&NiagaraSpriteRendererProperties, rendererIndex=1}
  ]
 }
 sparkle_center {
  emitterUpdateScript [
   &EmitterState(Inactive Response=::NewEnumerator1, Life Cycle Mode=::NewEnumerator1, Loop Behavior=::NewEnumerator1, Loop Duration=2, UseLoopDelay=false)
   &SpawnRate(SpawnRate=50, Use Spawn Probability=false) {enabled=false}
   &SpawnBurst_Instantaneous(Spawn Count=88, Use Spawn Probability=false, UseLoopCountLimit=false)
  ]
  particleSpawnScript [
   &InitializeParticle(
    Color Maximum=<rgba 1 0.5692 0.3873 1>
    Color Minimum=<rgba 1 0.6949 0.0328 1>
    Color Mode=::NewEnumerator2
    Lifetime=&RandomRangeFloat(Minimum=0.3, Maximum=1.2)
    Sprite Rotation Mode=::NewEnumerator2
    Sprite Size Mode=::NewEnumerator1
    Uniform Sprite Size Max=6
    Uniform Sprite Size Min=3.5
    UsePositionOffset=false
    Write Material Random=true
   )
   /SetVariables(Particles.RandomCore:NiagaraFloat=&FloatFromCurve(
     FloatCurve=DataInterface() {propertyValues {Curve {keys [
         {arriveTangent=0.0559, interpMode=RCIM_Cubic, leaveTangent=0.0559, tangentMode=RCTM_User, time=0, value=0.2}
         {arriveTangent=0.1895, interpMode=RCIM_Cubic, leaveTangent=0.1895, tangentMode=RCTM_User, time=0.9, value=0.25}
         {arriveTangent=1.1808, interpMode=RCIM_Cubic, leaveTangent=1.1808, tangentMode=RCTM_User, time=1, value=0.3}
        ]}}}
     CurveIndex=&RandomRangeFloat(Minimum=0, Maximum=1)
     Scale Curve=3
    )) {isSetParameters=true}
   &SphereLocation(
    MaskSpawnGroup=false
    Override Seed=false
    Sphere Radius=20
    Surface Only=false
    UseNonUniformScale=false
    UseOffset=false
    UseSphereOrientation=false
   )
   &AddVelocityFromPoint(Module.Use Falloff Distance=false, Velocity Strength=222)
   &AddVelocity(Velocity=<0 -80 0>) {enabled=false, module=AddVelocity001}
  ]
  particleUpdateScript [
   &ParticleState()
   &CurlNoiseForce(Noise Frequency=125, Noise Strength=&Multiply_Float(A=555, B=1), Pan Noise=true, Pan Noise Field=<0.5 0.5 0>)
   &ScaleSpriteSize(Scale Factor=&Vector2DFromFloat(Value=&FloatFromCurve(
      FloatCurve=DataInterface() {propertyValues {Curve {keys [
          <at 0 1 RCIM_Cubic>
          {arriveTangent=-1, interpMode=RCIM_Cubic, leaveTangent=-1, time=0.6, value=1.25}
          {arriveTangent=-0.1423, interpMode=RCIM_Cubic, leaveTangent=-0.1423, tangentMode=RCTM_User, time=1, value=0}
         ]}}}
      Scale Curve=$Particles.RandomCore
     )))
   &ScaleSpriteSize(
    Scale Factor=&Multiply_Vector2D(A=$Particles.SpriteSize, B=<0.3 0.3>)
    Scale Sprite Size Mode=::NewEnumerator2
   ) {module=ScaleSpriteSize001}
   &ScaleColor(
    Scale Alpha=&Multiply_Float(
     A=&FloatFromCurve(FloatCurve=DataInterface() {propertyValues {Curve {keys [<at 0.8 1>,<at 1 0>]}}})
     B=&FNTimeOfDayScalar(Night Scale=0.3, Morning Scale=0.1, Evening Scale=0.2)
    )
    Scale RGB=&VectorFromFloat(Value=&FloatFromCurve(FloatCurve=DataInterface() {propertyValues {Curve {keys [<at 0 1>,<at 0.1 2>,<at 0.2 1>]}}}))
    ScaleA=true
    ScaleRGB=true
   )
   &self:ScratchModule_08() {module=ScratchModule_08}
   &InheritSourceMovement(Applied Force Scale=<0 0 1>, Applied Position Scale=<0.8 0.8 0.8>, ApplyForce=false, ApplyPosition=true, ApplyPositionFalloff=false) {enabled=false, module=InheritSourceMovement001}
   &Drag(Use Linear Drag=true, Use Rotational Drag=false)
   &SolveForcesAndVelocity(Clamp Velocity=false, Limit Acceleration=false)
  ]
  rendererClasses [&NiagaraSpriteRendererProperties]
  renderers [{rendererClass=&NiagaraSpriteRendererProperties}]
 }
}
summary {
 systemName=NS_SquadLock_VaultOpen
 emitters [
  {bEnabled=false, emitterName=OmnidirectionalBurst, simTarget=GPUComputeSim, rendererClasses [&NiagaraSpriteRendererProperties]}
  {emitterName=LightRenderer, rendererClasses [&NiagaraLightRendererProperties]}
  {emitterName=Halo_and_Glow, rendererClasses [&NiagaraSpriteRendererProperties]}
  {emitterName=Shockwave_Flare, rendererClasses [&NiagaraMeshRendererProperties]}
  {emitterName=Distortion, rendererClasses [&NiagaraMeshRendererProperties]}
  {emitterName=TallBeam_sprite, rendererClasses [&NiagaraSpriteRendererProperties]}
  {emitterName=radial_flares, rendererClasses [&NiagaraSpriteRendererProperties]}
  {emitterName=single_halo001, rendererClasses [&NiagaraSpriteRendererProperties]}
  {emitterName=radial_flares001, rendererClasses [&NiagaraSpriteRendererProperties]}
  {emitterName=sparkle_center, rendererClasses [&NiagaraSpriteRendererProperties]}
  {emitterName=LockLight, rendererClasses [&NiagaraMeshRendererProperties]}
 ]
 userVariables [
  User.AgeInput {type=NiagaraFloat, defaultValue=0, description=""}
  User.CameraQuery {
   type=NiagaraDataInterfaceCamera
   description=""
   defaultValue {
    struct=NiagaraExt_VariableValue_DataInterface
    value {
     dataInterface=&self:NiagaraDataInterfaceCamera_0
     dataInterfaceClass=&NiagaraDataInterfaceCamera
    }
   }
  }
  User.SphereRadius {type=NiagaraFloat, defaultValue=0, description=""}
 ]
}
