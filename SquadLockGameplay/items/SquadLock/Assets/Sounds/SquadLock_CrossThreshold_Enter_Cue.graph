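# SquadLock sound cue for the threshold-crossing "enter" effect, likely heard when a player crosses into a squad lock's boundary.
# A single wave player feeds the output directly, routed through the gadgets sound class and Fortnite SFX submix.
# Inline attenuation with natural-sound falloff and stereo spread overrides the assigned explosion attenuation asset.
# Overridden concurrency permits one instance at a time and rejects new triggers while it plays.
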SoundCue /SquadLockGameplay/items/SquadLock/Assets/Sounds/SquadLock_CrossThreshold_Enter_Cue
root=WavePlayer_1
nodes [{id=WavePlayer_1, properties {soundWaveAssetPtr=/SquadLockGameplay/items/SquadLock/Assets/Sounds/wav/Squadlocks_PlayerXThreshold_Enter}}]
properties {
 attenuationSettings=/Game/Sounds/Attenuation/Fort_Explosion_Medium_StW_Att
 bOverrideAttenuation=true
 bOverrideConcurrency=true
 maxDistance=2700
 soundClassObject=&Fort_Gadgets
 soundSubmixObject=&Fortnite_SFX
 volumeMultiplier=0.997
 attenuationOverrides {distanceAlgorithm=NaturalSound, falloffDistance=2200, stereoSpread=500, attenuationShapeExtents {x=500}}
 concurrencyOverrides {maxCount=1, resolutionRule=PreventNew}
}
