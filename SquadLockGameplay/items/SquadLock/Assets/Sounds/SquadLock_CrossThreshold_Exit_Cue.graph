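# SquadLock sound cue that plays when a player crosses a threshold on exit, judging by its name.
# A single wave player feeds the exit wave straight to the output, boosted in volume, on the Gadgets sound class and Fortnite SFX submix.
# Attenuation is overridden in the cue with natural-sound falloff, despite a medium explosion attenuation asset also being referenced.
# Overridden concurrency caps simultaneous plays at two and rejects new ones beyond that.
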SoundCue /SquadLockGameplay/items/SquadLock/Assets/Sounds/SquadLock_CrossThreshold_Exit_Cue
root=WavePlayer_0
nodes [{id=WavePlayer_0, properties {soundWaveAssetPtr=/SquadLockGameplay/items/SquadLock/Assets/Sounds/wav/Squadlocks_PlayerXThreshold_Exit}}]
properties {
 attenuationSettings=/Game/Sounds/Attenuation/Fort_Explosion_Medium_StW_Att
 bOverrideAttenuation=true
 bOverrideConcurrency=true
 maxDistance=2400
 soundClassObject=&Fort_Gadgets
 soundSubmixObject=&Fortnite_SFX
 volumeMultiplier=1.396
 attenuationOverrides {distanceAlgorithm=NaturalSound, falloffDistance=2000}
 concurrencyOverrides {maxCount=2, resolutionRule=PreventNew}
}
