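# SquadLock gadget sound cue, named for the final lock visual, that plays a single wave after a fixed delay.
# A Delay node feeds one Wave Player for the FinalLockVisual wave, routed through the gadgets sound class and SFX submix.
# The cue overrides attenuation with a natural-sound falloff and stereo spread, alongside a referenced medium explosion attenuation asset.
# Its concurrency override allows one instance and blocks new plays while it sounds.
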SoundCue /SquadLockGameplay/items/SquadLock/Assets/Sounds/SquadLock_FinalLockVisual_Cue
root=Delay_1
nodes [
 {id=Delay_1, children [WavePlayer_0], properties {delayMax=1.75, delayMin=1.75}}
 {id=WavePlayer_0, properties {soundWaveAssetPtr=/SquadLockGameplay/items/SquadLock/Assets/Sounds/wav/Squadlocks_FinalLockVisual}}
]
properties {
 attenuationSettings=/Game/Sounds/Attenuation/Fort_Explosion_Medium_StW_Att
 bOverrideAttenuation=true
 bOverrideConcurrency=true
 maxDistance=4050
 soundClassObject=&Fort_Gadgets
 soundSubmixObject=&Fortnite_SFX
 volumeMultiplier=0.8
 attenuationOverrides {distanceAlgorithm=NaturalSound, nonSpatializedRadiusStart=60, stereoSpread=800, attenuationShapeExtents {x=450}}
 concurrencyOverrides {maxCount=1, resolutionRule=PreventNew}
}
