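# Sound cue for the SquadLock item that provides the beep named for its pending-unlock flash.
# A single wave player feeds the root directly, with no randomization or modulation.
# Overridden attenuation uses natural-sound falloff for a medium-range spatialized beep, routed through the ambient props sound class and SFX submix.
# Overridden concurrency allows one instance per owner and prevents new plays while one is active.
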SoundCue /SquadLockGameplay/items/SquadLock/Assets/Sounds/SquadLock_PendingUnlockFlash_Cue
root=WavePlayer_2
nodes [{id=WavePlayer_2, properties {soundWaveAssetPtr=/SquadLockGameplay/items/SquadLock/Assets/Sounds/wav/SquadLocks_PendingUnlockFlash_Beep}}]
properties {
 attenuationSettings=/Game/Sounds/Attenuation/Fort_Explosion_Medium_StW_Att
 bOverrideAttenuation=true
 bOverrideConcurrency=true
 maxDistance=3000
 soundClassObject=&Fort_Ambient_Props_Athena
 soundSubmixObject=&Fortnite_SFX
 volumeMultiplier=1.504
 attenuationOverrides {distanceAlgorithm=NaturalSound, falloffDistance=2200, nonSpatializedRadiusStart=60, stereoSpread=800, attenuationShapeExtents {x=800}}
 concurrencyOverrides {bLimitToOwner=true, maxCount=1, resolutionRule=PreventNew}
}
