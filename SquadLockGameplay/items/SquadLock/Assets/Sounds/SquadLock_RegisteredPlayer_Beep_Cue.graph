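# Sound cue for the SquadLock item's registered-player beep.
# A single wave player feeds the beep wave straight to the output at reduced volume, classed as an ambient prop sound and routed to the SFX submix.
# Inline overrides replace the assigned explosion attenuation with logarithmic distance falloff and limit concurrency to one instance, blocking new plays while one is active.
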SoundCue /SquadLockGameplay/items/SquadLock/Assets/Sounds/SquadLock_RegisteredPlayer_Beep_Cue
root=WavePlayer_0
nodes [{id=WavePlayer_0, properties {soundWaveAssetPtr=/SquadLockGameplay/items/SquadLock/Assets/Sounds/wav/Squadlocks_RegisteredPlayer_Beep}}]
properties {
 attenuationSettings=/Game/Sounds/Attenuation/Fort_Explosion_Medium_StW_Att
 bOverrideAttenuation=true
 bOverrideConcurrency=true
 maxDistance=3000
 soundClassObject=&Fort_Ambient_Props_Athena
 soundSubmixObject=&Fortnite_SFX
 volumeMultiplier=0.601
 attenuationOverrides {distanceAlgorithm=Logarithmic, falloffDistance=2200, attenuationShapeExtents {x=800}}
 concurrencyOverrides {maxCount=1, resolutionRule=PreventNew, retriggerTime=0.1}
}
