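# Sound Cue for the SquadLock scan power-up, likely the start of the scan loop, played as a positional prop sound.
# A single Wave Player plays the Squadlocks_Scan_PowerUp wave with boosted volume, routed to the Athena ambient props class and Fortnite SFX submix.
# Overridden attenuation adds natural-sound falloff, distance low-pass filtering and occlusion.
# Overridden concurrency caps it at two instances, stopping the oldest.
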SoundCue /SquadLockGameplay/items/SquadLock/Assets/Sounds/SquadLock_Scan_lp_Start_Cue
root=WavePlayer_0
nodes [{id=WavePlayer_0, properties {soundWaveAssetPtr=/SquadLockGameplay/items/SquadLock/Assets/Sounds/wav/Squadlocks_Scan_PowerUp}}]
properties {
 attenuationSettings=/Game/Sounds/Attenuation/Fort_Amb_Emitter_Prop_Med_Att
 bOverrideAttenuation=true
 bOverrideConcurrency=true
 maxDistance=2850
 soundClassObject=&Fort_Ambient_Props_Athena
 soundSubmixObject=&Fortnite_SFX
 volumeMultiplier=1.8
 attenuationOverrides {
  bAttenuateWithLPF=true
  bEnableOcclusion=true
  distanceAlgorithm=NaturalSound
  falloffDistance=2200
  lPFFrequencyAtMax=2000
  lPFFrequencyAtMin=10000
  lPFRadiusMax=1500
  lPFRadiusMin=500
  nonSpatializedRadiusStart=10
  occlusionLowPassFilterFrequency=5000
  occlusionTraceChannel=ECC_GameTraceChannel11
  reverbDistanceMax=2000
  stereoSpread=450
  attenuationShapeExtents {x=650}
 }
 concurrencyOverrides {maxCount=2, resolutionRule=StopOldest}
}
