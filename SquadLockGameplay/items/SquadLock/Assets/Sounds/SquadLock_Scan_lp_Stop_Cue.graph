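# Sound cue for the SquadLock item that plays the scan power-down sound, evidently when the looping scan stops.
# A single wave player feeds the root directly, routed to the ambient props sound class and the SFX submix.
# Overridden attenuation gives it natural falloff with distance low-pass filtering and occlusion, despite a referenced prop attenuation asset.
# Overridden concurrency caps simultaneous instances at a small count, stopping the oldest.
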SoundCue /SquadLockGameplay/items/SquadLock/Assets/Sounds/SquadLock_Scan_lp_Stop_Cue
root=WavePlayer_1
nodes [{id=WavePlayer_1, properties {soundWaveAssetPtr=/SquadLockGameplay/items/SquadLock/Assets/Sounds/wav/Squadlocks_Scan_PowerDown}}]
properties {
 attenuationSettings=/Game/Sounds/Attenuation/Fort_Amb_Emitter_Prop_Med_Att
 bOverrideAttenuation=true
 bOverrideConcurrency=true
 maxDistance=2950
 soundClassObject=&Fort_Ambient_Props_Athena
 soundSubmixObject=&Fortnite_SFX
 volumeMultiplier=1
 attenuationOverrides {
  bAttenuateWithLPF=true
  bEnableOcclusion=true
  distanceAlgorithm=NaturalSound
  falloffDistance=2500
  lPFFrequencyAtMax=2000
  lPFFrequencyAtMin=10000
  lPFRadiusMax=1500
  lPFRadiusMin=500
  nonSpatializedRadiusStart=50
  occlusionLowPassFilterFrequency=5000
  occlusionTraceChannel=ECC_GameTraceChannel11
  reverbDistanceMax=2000
  stereoSpread=650
  attenuationShapeExtents {x=450}
 }
 concurrencyOverrides {maxCount=2, resolutionRule=StopOldest}
}
